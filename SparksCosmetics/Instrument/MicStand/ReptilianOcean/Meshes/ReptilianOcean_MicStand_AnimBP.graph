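# Animation Blueprint for the ReptilianOcean mic stand cosmetic, playing an idle loop broken up by an occasional fidget.
# A state machine layers over the input pose from root_body via per-bone blend, returning to Idle automatically after Fidget_A.
# Fidgets start near the idle loop's end on a 20% per-update roll while IsSnakeIdle is false.
# The event graph cycles that flag: false for two seconds, then true for 3–10 random seconds.
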AnimBlueprint /SparksCosmetics/Instrument/MicStand/ReptilianOcean/Meshes/ReptilianOcean_MicStand_AnimBP
target_skeleton=&/ReptilianOcean_MicStand_Skeleton
graphs [
 AnimGraph {nodes [
   n1 = Random Sequence Player!() -> (Pose:Pose Link Structure) {
    is=title AnimGraphNode_RandomPlayer
    entries [
     {
      chanceToPlay=0.7
      maxLoopCount=0
      maxPlayRate=1
      minLoopCount=0
      minPlayRate=1
      sequence=null
      blendIn {blendOption=Linear, blendTime=0.2, customCurve=null}
     }
     {
      chanceToPlay=0.3
      maxLoopCount=0
      maxPlayRate=1
      minLoopCount=0
      minPlayRate=1
      sequence=null
      blendIn {blendOption=Linear, blendTime=0.2, customCurve=null}
     }
    ]
   }
   Apply Additive(Additive=@n1) {alphaBoolBlend {blendOption=Linear}}
   AnimGraphNode_StateMachine_0 = GameStateMachine() {is=title AnimGraphNode_StateMachine}
   Output Pose(Result=Layered blend per bone(BasePose=Input Pose(), BlendPoses_0=@AnimGraphNode_StateMachine_0) {layerSetup [{branchFilters [{boneName=root_body}]}]})
   Sequence Player 'ReptilianOcean_MicStand_Idle_Loop'() {sequence=&/ReptilianOcean_MicStand_Idle_Loop}
   n2 = Random Sequence Player() {entries [
     {
      chanceToPlay=1
      maxLoopCount=0
      maxPlayRate=1
      minLoopCount=0
      minPlayRate=1
      sequence=null
      blendIn {blendOption=Linear, blendTime=0.3, customCurve=null}
     }
     {
      chanceToPlay=1
      maxLoopCount=0
      maxPlayRate=1
      minLoopCount=0
      minPlayRate=1
      sequence=null
      blendIn {blendOption=Linear, blendTime=0.3, customCurve=null}
     }
    ]}
   Reroute Node(InputPin:Pose Link Structure=@n2) -> (OutputPin:Pose Link Structure)
   Reroute Node(InputPin:Wildcard) -> (OutputPin:Wildcard)
  ]}
 GameStateMachine {
  entry=Idle
  path=AnimGraph.AnimGraphNode_StateMachine_0.GameStateMachine
  ids=id
  nodes [
   AnimStateNode_3 = Idle() {
    stateEntered {guid=#1}
    stateFullyBlended {guid=#2}
    stateLeft {guid=#3}
   }
   AnimStateNode_4 = Idle Break() {
    is=title AnimStateNode
    stateEntered {guid=#4}
    stateFullyBlended {guid=#5}
    stateLeft {guid=#6}
   }
   AnimStateTransitionNode_2 = AnimStateTransitionNode()
   AnimStateTransitionNode_3 = AnimStateTransitionNode() {bAutomaticRuleBasedOnSequencePlayerInState=true}
  ]
  links [
   {
    dst=Idle Break
    rule=AnimGraph.AnimGraphNode_StateMachine_0.GameStateMachine.AnimStateTransitionNode_2.Transition
    src=Idle
    via=AnimStateTransitionNode_2
   }
   {
    dst=Idle
    rule=AnimGraph.AnimGraphNode_StateMachine_0.GameStateMachine.AnimStateTransitionNode_3.Transition
    src=Idle Break
    via=AnimStateTransitionNode_3
   }
  ]
 }
 Idle {
  path=AnimGraph.AnimGraphNode_StateMachine_0.GameStateMachine.AnimStateNode_3.Idle
  nodes [Output Animation Pose(Result=Sequence Player 'ReptilianOcean_MicStand_Idle_Loop'() {sequence=&ReptilianOcean_MicStand_Idle_Loop})]
 }
 Idle Break {
  path=AnimGraph.AnimGraphNode_StateMachine_0.GameStateMachine.AnimStateNode_4.Idle Break
  nodes [Output Animation Pose(Result=Sequence Player 'ReptilianOcean_MicStand_Fidget_A'() {bLoopAnimation=false, sequence=&/ReptilianOcean_MicStand_Fidget_A})]
 }
 Transition {
  path=AnimGraph.AnimGraphNode_StateMachine_0.GameStateMachine.AnimStateTransitionNode_2.Transition
  nodes [
   n1 = Get IsSnakeIdle() -> (IsSnakeIdle:bool)
   n2 = "Get Relevant Anim Time Remaining (Idle)"!() -> (ReturnValue:float) {is=title K2Node_AnimGetter}
   n3 = Get Random Int Sequence() -> (Random Int Sequence:int=0)
   Result(bCanEnterTransition=AND Boolean!(A="integer < integer"(A=@n3, B=20), B=NOT Boolean(A=@n1), C:bool="float <= float"(A=@n2, B=0.2) -> (ReturnValue=~)) -> (ReturnValue))
  ]
 }
 Transition {path=AnimGraph.AnimGraphNode_StateMachine_0.GameStateMachine.AnimStateTransitionNode_3.Transition, nodes [Result(bCanEnterTransition=true)]}
 EventGraph {nodes [
   Event BlueprintUpdateAnimation()
   > Set Random Int Sequence(Random Int Sequence:int=RandomInteger!(Max:int=100) -> (ReturnValue:int=0)) -> (Output_Get:int=0)
   > Do Once(Reset=@n1.then)
   > Set IsSnakeIdle(IsSnakeIdle:bool=false) -> (Output_Get:bool)
   > Delay(Duration=2)
   > Set IsSnakeIdle(IsSnakeIdle=true)
   > n1 = Delay(Duration=RandomFloatInRange(Min=3, Max=10))
   TryGetPawnOwner()
  ]}
]
variables [
 IsSnakeIdle {type=boolean}
 Random Int Sequence {type=integer}
]
